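; BP_TrickTile is a Creative trap device that makes its attached building tile vanish when valid players touch it or event receivers trigger it.
; A replicated ForcedTileState moves through delayed activation, destroyed and restored states, toggling collision, visibility and damageability, with reset by timer, event or minigame end.
; Server and clients branch on state changes.
; During activation, a client timeline drives staged glow and bounce feedback, limited on mobile.
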(event EventOnPlaced
  ;[0 1 0.1052 0.5] Initialization
  (Networking|CanExecuteCosmeticEvents
    (:True
      (Audio|PlaySoundatLocation (Variables|Audio|GetPlace_Trap_Sound) (Transformation|GetActorLocation)))))

(event Attachment|EventOnAttachToBuilding
  (params (AttachedToBuilding :type Building%20Actor%20Object%20Reference))
  ;[0 1 0.1052 0.5] Initialization
  ;;[1 1 1 0.5] Handle possible racing condition, between the replication of the attach and the forced tile state
  (bind _asbuilding_static_mesh (Utilities|Casting|CastToBuildingSMActor AttachedToBuilding))
  (Variables|Default|SetBaseTileCached _asbuilding_static_mesh)
  (Networking|SwitchHasAuthority
    (:Remote
      (if (Utilities|Enum|NotEqual(Enum) (Variables|Replicated|GetForcedTileState) "NewEnumerator0")
        (TileState|OnTileStateChange)))))

(event EventBeginPlay
  ;[0 1 0.1052 0.5] Initialization
  (|CallParentFunction)
  (Networking|SwitchHasAuthority
    (:Authority
      (CallFunction|InitializeDevice))
    (:Remote
      (TileState|OnTileStateChange))))

(event OnAnyPropertyChanged(ToyOptionsComponent)
  ;[0 1 0.1052 0.5] Initialization
  (CallFunction|InitializeDevice))

(event Custom|Initialize%20Device
  ;[0 1 0.1052 0.5] Initialization
  (CallFunction|MigrateSameChannelInstances)
  (CallFunction|InitializeComponents)
  (CallFunction|CheckFeedbackFeasibility)
  (Networking|CanExecuteCosmeticEvents
    (:True
      (VFX|UpdateActivationVFXTimes))))

(event OnPropertiesLoaded(ToyOptionsComponent)
  ;[0 1 0.1052 0.5] Initialization
  (CallFunction|InitializeDevice))

(event EventDestroyed
  ;[1 0.04377 0] Destruction
  (if (Utilities|Enum|Equal(Enum) (Variables|Replicated|GetForcedTileState) "NewEnumerator3")
    (Utilities|Time|ClearandInvalidateTimerbyHandle (Variables|Default|GetActivationTimerHandler))
    (Utilities|Time|ClearandInvalidateTimerbyHandle (Variables|Default|GetWaitingForDestroyAnimationFinishHandle))
    (Utilities|Time|ClearandInvalidateTimerbyHandle (Variables|Default|GetResetAfterTimerHandle))))

(event OnComponentBeginOverlap(TriggerBox)
  (params (OverlappedComponent :type Primitive%20Component%20Object%20Reference) (OtherActor :type Actor%20Object%20Reference) (OtherComp :type Primitive%20Component%20Object%20Reference) (OtherBodyIndex :type Integer) (bFromSweep :type Boolean) (SweepResult :type Hit%20Result%20Structure%20%28by%20ref%29))
  ;[1 1 1] Handle Overlap
  (Utilities|IsValid (Variables|Default|GetBaseTileCached)
    (:Is%20Valid
      (Utilities|IsValid OtherActor
        (:Is%20Valid
          (if (Trigger|IsActorValidtoTrigger OtherActor)
            (bind (_valid_controller _fortplayerpawn _playercontroller) (DEPRECATED|ActorHasValidController (Variables|Default|GetCreative_UseRestriction_Component) OtherActor true true))
            (if _valid_controller
              (Trigger|TryTriggerTrap _playercontroller))))))))

(event Custom|Trigger%20Event
  (params (TriggerInstigator :type Controller%20Object%20Reference))
  ;[0.5229 0 1 0.5] Receivers
  (Trigger|TryTriggerTrap 0))

(event PawnNowHasController(Creative_UseRestriction_Component)
  (params (FortPlayerPawn :type Fort%20Player%20Pawn%20Object%20Reference) (Player%20Controller :type Fort%20Player%20Controller%20Object%20Reference))
  ;[1 1 1] Handle Overlap
  (if (Collision|IsOverlappingActor (Variables|Default|GetTriggerBox) FortPlayerPawn :!class /Script/Engine.PrimitiveComponent)
    (Trigger|TryTriggerTrap Player%20Controller)))

(event Custom|Reset%20Event
  (params (Instigator :type Controller%20Object%20Reference))
  ;[0.5229 0 1 0.5] Receivers
  (CallFunction|TryResetTile))

(event Custom|Enable%20Event
  (params (Instigator :type Controller%20Object%20Reference))
  ;[0.5229 0 1 0.5] Receivers
  (Class|CreativeEnabledComponent|SetEnabled (Variables|Default|GetCreative_Enabled_Component) true))

(event Custom|Disable%20Event
  (params (Instigator :type Controller%20Object%20Reference))
  ;[0.5229 0 1 0.5] Receivers
  (Class|CreativeEnabledComponent|SetEnabled (Variables|Default|GetCreative_Enabled_Component) false))

(event Custom|Toggle%20Event
  (params (Instigator :type Controller%20Object%20Reference))
  ;[0.5229 0 1 0.5] Receivers
  (Class|CreativeEnabledComponent|SetEnabled (Variables|Default|GetCreative_Enabled_Component) (not (Class|CreativeEnabledComponent|IsEnabled (Variables|Default|GetCreative_Enabled_Component)))))

(event Custom|Enable%20Player%20Contact
  (params (Instigator :type Controller%20Object%20Reference))
  ;[0.5229 0 1 0.5] Receivers
  (if (not (Variables|Replicated|GetDesiredTriggerState))
    (Variables|Replicated|SetDesiredTriggerState true)
    (Trigger|ServerUpdateTrigger)))

(event Custom|Disable%20Player%20Contact
  (params (Instigator :type Controller%20Object%20Reference))
  ;[0.5229 0 1 0.5] Receivers
  (if (Variables|Replicated|GetDesiredTriggerState)
    (Variables|Replicated|SetDesiredTriggerState false)
    (Trigger|ServerUpdateTrigger)))

(event Custom|Toggle%20Player%20Contect
  (params (Instigator :type Controller%20Object%20Reference))
  ;[0.5229 0 1 0.5] Receivers
  (Variables|Replicated|SetDesiredTriggerState (not (Variables|Replicated|GetDesiredTriggerState)))
  (Trigger|ServerUpdateTrigger))

(event MinigameStartedDelegate(FortMinigameProgress)
  ;[1 1 1] Handle Minigame events
  (CallFunction|CacheOriginalValuesfromBaseTile)
  (Networking|SwitchHasAuthority
    (:Authority
      (Variables|Replicated|SetDesiredTriggerState (Math|Conversions|ToBoolean(Integer) (Variables|UserKnob|GetTriggerOnPlayerContactKnob)))
      (Trigger|ServerUpdateTrigger)
      (if (Math|Conversions|ToBoolean(Integer) (Variables|UserKnob|GetActivateAtGameStartKnob))
        (TileState|SetForcedTileState "NewEnumerator2")))))

(event MinigameEndedDelegate(FortMinigameProgress)
  ;[1 1 1] Handle Minigame events
  (Utilities|Time|ClearandInvalidateTimerbyHandle (Variables|Default|GetResetAfterTimerHandle))
  (TileState|RestoreTiletoInitialState))

(event Custom|OnDestroyAnimationFinish
  (CallFunction|SetDestroyedVisibility)
  (Utilities|Time|InvalidateTimerHandle (Variables|Default|GetWaitingForDestroyAnimationFinishHandle)))

(event Custom|StartVFXTimeline
  (bind _activation_stage (Variables|VFX|GetActivation%20Stage))
  (bind _vfxdataperstage (Variables|VFX|GetVFXDataPerStage))
  (bind (_output_glowcolor_8_8eb4076e423a62fecc95eda562d126d1 _output_glowfrequency_11_86f5d51e4a81205b5bc1f28583856721) (Utilities|Struct|BreakStructTrickTileVFXData (Utilities|Array|Get(aref) _vfxdataperstage _activation_stage)))
  (bind (_output_glowcolor_8_8eb4076e423a62fecc95eda562d126d1_r _output_glowcolor_8_8eb4076e423a62fecc95eda562d126d1_g _output_glowcolor_8_8eb4076e423a62fecc95eda562d126d1_b _output_glowcolor_8_8eb4076e423a62fecc95eda562d126d1_a) (Math|Color|BreakColor _output_glowcolor_8_8eb4076e423a62fecc95eda562d126d1))
  (bind _glowintensity (CallFunction|GetGlowIntensitybyCurrentActivationTime _output_glowcolor_8_8eb4076e423a62fecc95eda562d126d1_a _output_glowfrequency_11_86f5d51e4a81205b5bc1f28583856721))
  (bind (_returnvalue_r _returnvalue_g _returnvalue_b _returnvalue_a) (Math|Color|BreakColor (Math|Color|LinearColor*Float (Math|Color|MakeColor _output_glowcolor_8_8eb4076e423a62fecc95eda562d126d1_r _output_glowcolor_8_8eb4076e423a62fecc95eda562d126d1_g _output_glowcolor_8_8eb4076e423a62fecc95eda562d126d1_b _output_glowcolor_8_8eb4076e423a62fecc95eda562d126d1_a) _glowintensity)))
  (bind (_output_glowcolor_8_8eb4076e423a62fecc95eda562d126d1_2 _output_glowfrequency_11_86f5d51e4a81205b5bc1f28583856721_2 _output_bounceintensity_5_3e346d374b936ec30283a490ff3172c7 _output_bouncefrequency_7_15970f9041dde520322e1bafc89fba97) (Utilities|Struct|BreakStructTrickTileVFXData (Utilities|Array|Get(aref) (Variables|VFX|GetVFXDataPerStage) (Variables|VFX|GetActivation%20Stage))))
  (Timeline|VFXTimeline PlayFromStart
    (:Update
      (Variables|VFX|SetCurrentActivationTime (Math|Float|Min(Float) (* (Components|Timeline|GetPlaybackPosition (Variables|BP_TrickTile|GetVFXTimeline)) (Variables|UserKnob|GetActivationDelayKnob)) (Variables|UserKnob|GetActivationDelayKnob)))
      (Utilities|FlowControl|Sequence
        (:then_0
          (CallFunction|SetActivationStatebyCurrentActivationTime))
        (:then_1
          (if (or (Utilities|Enum|Equal(Enum) (Variables|UserKnob|GetTriggerFeedback) "NewEnumerator0") (Utilities|Enum|Equal(Enum) (Variables|UserKnob|GetTriggerFeedback) "NewEnumerator2"))
            (bind _basetilecached (Variables|Default|GetBaseTileCached)
              (:then
                (if (CallFunction|CanDoBuildingAnimations _basetilecached)
                  (Rendering|Material|SetCustomPrimitiveDataVector4 (Building|GetBuildingMeshComponent (Variables|Default|GetBaseTileCached)) 7 (Math|Vector4|MakeVector4 _returnvalue_r _returnvalue_g _returnvalue_b _returnvalue_a))
                  (else
                    (for _array_element (Rendering|Material|GetMaterials (Building|GetBuildingMeshComponent (Variables|Default|GetBaseTileCached)))
                      (bind _asmaterial_instance_dynamic (Utilities|Casting|CastToMaterialInstanceDynamic _array_element))
                      (Rendering|Material|SetVectorParameterValue _asmaterial_instance_dynamic "BlueprintColor" (Math|Color|MakeColor _returnvalue_r _returnvalue_g _returnvalue_b _returnvalue_a) :!class /Script/Engine.MaterialInstanceDynamic)))))
              (:else))))
        (:then_2
          (if (or (Utilities|Enum|Equal(Enum) (Variables|UserKnob|GetTriggerFeedback) "NewEnumerator0") (Utilities|Enum|Equal(Enum) (Variables|UserKnob|GetTriggerFeedback) "NewEnumerator1"))
            (if (!= (Variables|VFX|GetHitAnimTimer) -1.0)
              (bind _output_get (Variables|VFX|SetHitAnimTimer (+ (Utilities|Time|GetWorldDeltaSeconds) (Variables|VFX|GetHitAnimTimer))))
              (if (>= _output_get _output_bouncefrequency_7_15970f9041dde520322e1bafc89fba97)
                (VFX|PlayCustomHitBounceAnimation)))))))))

(event Custom|StopVFXTimeline
  (bind _activation_stage (Variables|VFX|GetActivation%20Stage))
  (bind _vfxdataperstage (Variables|VFX|GetVFXDataPerStage))
  (bind (_output_glowcolor_8_8eb4076e423a62fecc95eda562d126d1 _output_glowfrequency_11_86f5d51e4a81205b5bc1f28583856721) (Utilities|Struct|BreakStructTrickTileVFXData (Utilities|Array|Get(aref) _vfxdataperstage _activation_stage)))
  (bind (_output_glowcolor_8_8eb4076e423a62fecc95eda562d126d1_r _output_glowcolor_8_8eb4076e423a62fecc95eda562d126d1_g _output_glowcolor_8_8eb4076e423a62fecc95eda562d126d1_b _output_glowcolor_8_8eb4076e423a62fecc95eda562d126d1_a) (Math|Color|BreakColor _output_glowcolor_8_8eb4076e423a62fecc95eda562d126d1))
  (bind _glowintensity (CallFunction|GetGlowIntensitybyCurrentActivationTime _output_glowcolor_8_8eb4076e423a62fecc95eda562d126d1_a _output_glowfrequency_11_86f5d51e4a81205b5bc1f28583856721))
  (bind (_returnvalue_r _returnvalue_g _returnvalue_b _returnvalue_a) (Math|Color|BreakColor (Math|Color|LinearColor*Float (Math|Color|MakeColor _output_glowcolor_8_8eb4076e423a62fecc95eda562d126d1_r _output_glowcolor_8_8eb4076e423a62fecc95eda562d126d1_g _output_glowcolor_8_8eb4076e423a62fecc95eda562d126d1_b _output_glowcolor_8_8eb4076e423a62fecc95eda562d126d1_a) _glowintensity)))
  (bind (_output_glowcolor_8_8eb4076e423a62fecc95eda562d126d1_2 _output_glowfrequency_11_86f5d51e4a81205b5bc1f28583856721_2 _output_bounceintensity_5_3e346d374b936ec30283a490ff3172c7 _output_bouncefrequency_7_15970f9041dde520322e1bafc89fba97) (Utilities|Struct|BreakStructTrickTileVFXData (Utilities|Array|Get(aref) (Variables|VFX|GetVFXDataPerStage) (Variables|VFX|GetActivation%20Stage))))
  (Timeline|VFXTimeline Stop))

(fn CacheOriginalValuesFromBaseTile
  (bind _basetilecached (Variables|Default|GetBaseTileCached)
    (:then
      (Variables|Default|SetOriginalHealth (Health|GetHealth _basetilecached))
      (Variables|Default|SetOriginalSupressHealthBarValue (Variables|/Script/FortniteGame.BuildingActor|GetbSurpressHealthBar _basetilecached)))
    (:else)))

(fn CheckFeedbackFeasibility
  (if (and (TouchPlatform|IsMobileGame) (Math|Byte|NotEqual(Byte) (Variables|UserKnob|GetTriggerFeedback) (Utilities|Enum|LiteralenumEnum_TrickTile_TriggerFeedback "NewEnumerator3")))
    (Variables|UserKnob|SetTriggerFeedback "NewEnumerator2")))

(fn ClearActivation
  (params (bForced :type Boolean))
  (bind _basetilecached (Variables|Default|GetBaseTileCached))
  (bind _returnvalue (Building|GetBuildingMeshComponent _basetilecached))
  (bind _activationtimerhandler (Variables|Default|GetActivationTimerHandler))
  (if bForced
    (Utilities|Time|ClearandInvalidateTimerbyHandle _activationtimerhandler)
    (else
      (Utilities|Time|InvalidateTimerHandle _activationtimerhandler)))
  (Networking|CanExecuteCosmeticEvents
    (:True
      (CallFunction|StopVFXTimeline)
      (if (or (Utilities|Enum|Equal(Enum) (Variables|UserKnob|GetTriggerFeedback) "NewEnumerator0") (Utilities|Enum|Equal(Enum) (Variables|UserKnob|GetTriggerFeedback) "NewEnumerator2"))
        (bind _basetilecached_2 (Variables|Default|GetBaseTileCached)
          (:then
            (if (CallFunction|CanDoBuildingAnimations _basetilecached_2)
              (Rendering|Material|SetCustomPrimitiveDataVector4 _returnvalue 7 (0.0 0.0 0.0 0.0))
              (else
                (for _array_element (Rendering|Material|GetMaterials _returnvalue)
                  (bind _asmaterial_instance_dynamic (Utilities|Casting|CastToMaterialInstanceDynamic _array_element))
                  (Rendering|Material|SetVectorParameterValue _asmaterial_instance_dynamic "TimestampPayload" (0.0 0.0 0.0 1.0) :!class /Script/Engine.MaterialInstanceDynamic)))))
          (:else))))))

(fn DisableTileCanBeDamaged
  (bind _basetilecached (Variables|Default|GetBaseTileCached))
  (if (Variables|/Script/Engine.Actor|GetbCanBeDamaged _basetilecached)
    (Fortnite|SetCanbeDamaged _basetilecached)
    (Variables|Default|SetOverrideCanBeDamaged true)))

(fn GetGlowIntensityByCurrentActivationTime
  (params (GlowAlpha :type Float%20%28double-precision%29) (GlowFrequency :type Float%20%28double-precision%29))
  (return (* GlowAlpha (+ (Math|Trig|Sin(Radians) (+ (Math|Float|float*float (* GlowFrequency (Variables|VFX|GetCurrentActivationTime)) (Math|Trig|GetPI) 2.0) (Variables|VFX|GetOffset))) 1.3))))

(fn GetHitNormalForBounce
  (return "0, 0,1.000000"))

(fn GetVisualComponents
  (return (Utilities|Array|MakeArray (Variables|Debug|GetStaticMeshComponent))))

(fn InitializeComponents
  (Class|CreativeEnabledComponent|InitializeComponent (Variables|Default|GetCreative_Enabled_Component) (Variables|Default|GetFortMinigameProgress) (Math|Integer|SelectInt 2 -1 (Math|Conversions|ToBoolean(Integer) (Variables|UserKnob|GetEnabledAtGameStartKnob%20))))
  (Initialization|InitializeComponent (Variables|Default|GetCreative_UseRestriction_Component) (Variables|Default|GetFortMinigameProgress) (Variables|Default|GetCreative_Enabled_Component) 0 (Variables|UserKnob|GetSafeTeamKnob) (Variables|UserKnob|GetClassIndex) 0 1 :!class /Game/Creative/Devices/Common/Components/Creative_UseRestriction_Component.Creative_UseRestriction_Component_C))

(fn IsVisibleDuringMinigame
  (return false))

(fn IsActorValidToTrigger
  (params (ActorTryingToTrigger :type Actor%20Object%20Reference))
  (bind _aspawn (Utilities|Casting|CastToPawn ActorTryingToTrigger)
    (:then
      (Networking|SwitchHasAuthority
        (:Authority
          _)
        (:Remote
          (if (not (Pawn|IsLocallyControlled _aspawn))
            (return false))))
      (return true))
    (:CastFailed
      (return false))))

(fn MigrateSameChannelInstances
  (bind _bwasmigrated (UEFNConversionHelper|MigratetoReceiverwhenSameChannel (Variables|Default|GetEnableWhenReceivingFrom) (Variables|Default|GetDisableWhenReceivingFrom) (Variables|Default|GetToggleWhenReceivingFrom)))
  (bind _bwasmigrated_2 (UEFNConversionHelper|MigratetoReceiverwhenSameChannel (Variables|Default|GetEnablePlayerContactTriggerWhenReceivingFrom) (Variables|Default|GetDisablePlayerContactTriggerWhenReceivingFrom) (Variables|Default|GetTogglePlayerContactTriggerWhenReceivingFrom)))
  (return (and _bwasmigrated_2 _bwasmigrated)))

(fn OnActivationFinish
  (Activation|ClearActivation false)
  (TileState|SetForcedTileState "NewEnumerator2"))

(fn OnRep_CanTriggerOnPlayerContact
  (Collision|SetCollisionEnabled (Variables|Default|GetTriggerBox) (select (Variables|Replicated|GetCanTriggerOnPlayerContact) "QueryOnly" "NoCollision")))

(fn OnRep_ForcedTileState
  (Networking|SwitchHasAuthority
    (:Remote
      (if (Utilities|Enum|NotEqual(Enum) (Variables|Default|GetOldForcedTileState) (Variables|Replicated|GetForcedTileState))
        (TileState|OnTileStateChange)))))

(fn OnTileStateChange
  (bind _basetilecached (Variables|Default|GetBaseTileCached))
  (bind _basetilecached_2 (Variables|Default|GetBaseTileCached))
  (bind _basetilecached_3 (Variables|Default|GetBaseTileCached))
  (if (Utilities|Time|IsValidTimerHandle (Variables|Default|GetActivationTimerHandler))
    (Activation|ClearActivation true)
    (else
      _))
  (Utilities|FlowControl|Sequence
    (:then_0
      (Variables|Default|GetBaseTileCached
        (:then
          (Networking|SwitchHasAuthority
            (:Authority
              (switch Utilities|FlowControl|Switch|SwitchonEnum_TrickTileState (Variables|Replicated|GetForcedTileState)
                (:NewEnumerator1
                  (CallFunction|SetRestoredCollision)
                  (CallFunction|RestoreDamagetoBaseTile)
                  (Variables|Trigger|SetTrap%20Instigator 0) ; Reset instigator when the tile is restored.
                  (Utilities|Time|SetTimerbyEvent (EventDispatchers|CreateEvent :!event Server%20Update%20Trigger) (Variables|Trigger|GetReset%20Overlap%20Recheck%20Time)) ; After a short delay (to allow for Clients to see the wall \nrebuilding), check to see if anyone is standing on it and \ntrigger if they are.
                  (if (CallFunction|CanDoBuildingAnimations (Variables|Default|GetBaseTileCached))
                    (CallFunction|ForceVisualState (Variables|Default|GetBaseTileCached) "EBA_Placement")))
                (:NewEnumerator2
                  (Default|SetDestroyedCollision)
                  (GameplayEvent|BroadcastEvent (Variables|UserOptions|Events|GetOn%20Activated) 0 (Utilities|Struct|MakeGameplayEventActivated (Variables|Trigger|GetTrap%20Instigator)))
                  (CallFunction|RemoveDamagefromBaseTile)
                  (if (> (Variables|UserKnob|GetResetAfterKnob) 0)
                    (bind _returnvalue (Utilities|Time|SetTimerbyEvent (EventDispatchers|CreateEvent :!event TryResetTile) (Math|Conversions|ToFloat(Integer) (Variables|UserKnob|GetResetAfterKnob))))
                    (Variables|Default|SetResetAfterTimerHandle _returnvalue)
                    (else
                      _))
                  (if (CallFunction|CanDoBuildingAnimations (Variables|Default|GetBaseTileCached))
                    (Utilities|FlowControl|Sequence
                      (:then_0
                        (if (Building|IsUnderConstruction _basetilecached)
                          (CallFunction|CancelAutoBuild _basetilecached)))
                      (:then_1
                        (CallFunction|ForceVisualState _basetilecached_3 "EBA_Destruction")
                        (Networking|ForceNetUpdate _basetilecached_3)
                        (CallFunction|DisableTileCanbeDamaged)))))
                (:NewEnumerator3
                  (Activation|StartActivation))))
            (:Remote
              (switch Utilities|FlowControl|Switch|SwitchonEnum_TrickTileState (Variables|Replicated|GetForcedTileState)
                (:NewEnumerator1
                  (CallFunction|SetRestoredCollision)
                  (CallFunction|SetSuppressHealthBar _basetilecached_2 (Variables|Default|GetOriginalSupressHealthBarValue))
                  (Default|SetRestoredVisibility)
                  (Utilities|Time|ClearandInvalidateTimerbyHandle (Variables|Default|GetWaitingForDestroyAnimationFinishHandle))
                  (if (not (CallFunction|CanDoBuildingAnimations (Variables|Default|GetBaseTileCached)))
                    (CallFunction|PlayConstructionBounce (Variables|Default|GetBaseTileCached))))
                (:NewEnumerator2
                  (Default|SetDestroyedCollision)
                  (CallFunction|SetSuppressHealthBar _basetilecached_2 true)
                  (CallFunction|PlayDestructionEffects (Variables|Default|GetBaseTileCached))
                  (if (not (CallFunction|CanDoBuildingAnimations (Variables|Default|GetBaseTileCached)))
                    (CallFunction|SetDestroyedVisibility)
                    (Utilities|Time|ClearandInvalidateTimerbyHandle (Variables|Default|GetWaitingForDestroyAnimationFinishHandle))
                    (else
                      (bind _returnvalue_2 (Utilities|Time|SetTimerbyEvent (EventDispatchers|CreateEvent :!event OnDestroyAnimationFinish) 2.0))
                      (Variables|Default|SetWaitingForDestroyAnimationFinishHandle _returnvalue_2))))
                (:NewEnumerator3
                  (Activation|StartActivation))))))
        (:else)))
    (:then_1
      (Variables|Default|SetOldForcedTileState (Variables|Replicated|GetForcedTileState)))))

(fn PlayCustomHitBounceAnimation
  (bind (_output_glowcolor_8_8eb4076e423a62fecc95eda562d126d1 _output_glowfrequency_11_86f5d51e4a81205b5bc1f28583856721 _output_bounceintensity_5_3e346d374b936ec30283a490ff3172c7) (Utilities|Struct|BreakStructTrickTileVFXData (Utilities|Array|Get(aref) (Variables|VFX|GetVFXDataPerStage) (Variables|VFX|GetActivation%20Stage))))
  (Variables|VFX|SetHitAnimTimer 0.0)
  (bind _basetilecached (Variables|Default|GetBaseTileCached)
    (:then
      (bind (_origin _boxextent _sphereradius) (Collision|GetComponentBounds (Building|GetBuildingMeshComponent _basetilecached)))
      (bind (_origin_x _origin_y) (Math|Vector|BreakVector _origin))
      (bind (_returnvalue_x _returnvalue_y _returnvalue_z) (Math|Vector|BreakVector (Transformation|GetActorLocation _basetilecached)))
      (CallFunction|StartBounceAnimation _basetilecached (Utilities|Time|GetTimeSeconds) (Math|Float|Min(Float) _sphereradius 3e+02) (* (VFX|GetHitNormalforBounce) (* _output_bounceintensity_5_3e346d374b936ec30283a490ff3172c7 (select (CallFunction|CanDoBuildingAnimations (Variables|Default|GetBaseTileCached)) (Variables|VFX|GetMaxBounceFactorForAnimated) (Variables|VFX|GetMaxBounceFactorForCommon)))) (Math|Vector|MakeVector _origin_x _origin_y (+ _returnvalue_z (select (Utilities|Enum|Equal(Enum) (Variables|Attachment|GetBuildingAttachmentType) "ATTACH_Wall") 128.0 0.0))) "Hit" false))
    (:else)))

(fn RemoveDamageFromBaseTile
  (bind _basetilecached (Variables|Default|GetBaseTileCached)
    (:then
      (Variables|Default|SetOriginalHealth (Health|GetHealth _basetilecached))
      (Health|SetHealth _basetilecached (Health|GetMaxHealth _basetilecached)))
    (:else)))

(fn RestoreDamageToBaseTile
  (bind _basetilecached (Variables|Default|GetBaseTileCached)
    (:then
      (Health|SetHealth _basetilecached (Variables|Default|GetOriginalHealth)))
    (:else)))

(fn RestoreTileCanBeDamaged
  (if (Variables|Default|GetOverrideCanBeDamaged)
    (Fortnite|SetCanbeDamaged (Variables|Default|GetBaseTileCached) true)
    (Variables|Default|SetOverrideCanBeDamaged false)))

(fn RestoreTileToInitialState
  (if (Utilities|Enum|NotEqual(Enum) (Variables|Replicated|GetForcedTileState) "NewEnumerator0")
    (LocalVariables|SetTileStateBeforeRestore (Variables|Replicated|GetForcedTileState))
    (TileState|SetForcedTileState "NewEnumerator0")
    (Utilities|IsValid (Variables|Default|GetBaseTileCached)
      (:Is%20Valid
        (if (Utilities|Enum|Equal(Enum) (LocalVariables|GetTileStateBeforeRestore) "NewEnumerator2")
          (CallFunction|SetRestoredCollision)
          (Default|SetRestoredVisibility)
          (CallFunction|RestoreTileCanbeDamaged)
          (if (CallFunction|CanDoBuildingAnimations (Variables|Default|GetBaseTileCached))
            (CallFunction|ForceVisualState (Variables|Default|GetBaseTileCached) "EBA_Placement")
            (Utilities|Time|ClearandInvalidateTimerbyHandle (Variables|Default|GetWaitingForDestroyAnimationFinishHandle))))))))

(fn ServerUpdateTrigger
  (Networking|SwitchHasAuthority
    (:Authority
      (if (or (Utilities|Enum|Equal(Enum) (Variables|Replicated|GetForcedTileState) "NewEnumerator0") (Utilities|Enum|Equal(Enum) (Variables|Replicated|GetForcedTileState) "NewEnumerator1"))
        (Variables|Replicated|SetCanTriggerOnPlayerContact (Variables|Replicated|GetDesiredTriggerState))))))

(fn SetActivationStateByCurrentActivationTime
  (Utilities|FlowControl|Sequence
    (:then_0
      (if (== (Variables|VFX|GetActivation%20Stage) 0)
        (Math|Float|CompareFloat (Variables|VFX|GetCurrentActivationTime) (Variables|VFX|GetStageOneTime)
          (:>
            (Math|Integer|IncrementInt (Variables|VFX|GetActivation%20Stage))))))
    (:then_1
      (if (== (Variables|VFX|GetActivation%20Stage) 1)
        (Math|Float|CompareFloat (Variables|VFX|GetCurrentActivationTime) (Variables|VFX|GetStageTwoTime)
          (:>
            (Math|Integer|IncrementInt (Variables|VFX|GetActivation%20Stage))))))))

(fn SetDestroyedCollision
  (bind _basetilecached (Variables|Default|GetBaseTileCached)
    (:then
      (if (Utilities|DoesObjectImplementInterface _basetilecached "/Game/Creative/Devices/TrickTile/TrickTileInterface.TrickTileInterface_C")
        (bind _returnvalue (TrickTile|GetStaticMeshComponents(Message) (Variables|Default|GetBaseTileCached)))
        (for _array_element _returnvalue
          (Utilities|IsValid _array_element
            (:Is%20Valid
              (Collision|SetCollisionEnabled _array_element "NoCollision"))))
        (else
          (Collision|SetCollisionEnabled (Variables|/Script/FortniteGame.BuildingSMActor|GetStaticMeshComponent (Variables|Default|GetBaseTileCached)) "NoCollision"))))
    (:else)))

(fn SetDestroyedVisibility
  (bind _basetilecached (Variables|Default|GetBaseTileCached)
    (:then
      (if (Utilities|DoesObjectImplementInterface _basetilecached "/Game/Creative/Devices/TrickTile/TrickTileInterface.TrickTileInterface_C")
        (bind _returnvalue (TrickTile|GetStaticMeshComponents(Message) (Variables|Default|GetBaseTileCached)))
        (for _array_element _returnvalue
          (Utilities|IsValid _array_element
            (:Is%20Valid
              (Rendering|SetVisibility _array_element false false))))
        (else
          (Rendering|SetVisibility (Variables|/Script/FortniteGame.BuildingSMActor|GetStaticMeshComponent (Variables|Default|GetBaseTileCached)) false false))))
    (:else)))

(fn SetForcedTileState
  (params (NewState :type Enum_TrickTileState%20Enum :default "NewEnumerator0"))
  (if (Utilities|Enum|NotEqual(Enum) NewState (Variables|Replicated|GetForcedTileState))
    (Variables|Replicated|SetForcedTileState NewState)
    (TileState|OnTileStateChange)))

(fn SetRestoredCollision
  (bind _meshcomponent (LocalVariables|GetMeshComponent))
  (bind _defaultcollisions (LocalVariables|GetDefaultCollisions))
  (bind (_value _returnvalue) (Utilities|Map|Find _defaultcollisions _meshcomponent))
  (bind _meshcomponent_2 (LocalVariables|GetMeshComponent))
  (bind _basetilecached (Variables|Default|GetBaseTileCached)
    (:then
      (if (Utilities|DoesObjectImplementInterface _basetilecached "/Game/Creative/Devices/TrickTile/TrickTileInterface.TrickTileInterface_C")
        (bind _returnvalue_2 (TrickTile|GetDefaultCollisions(Message) (Variables|Default|GetBaseTileCached)))
        (LocalVariables|SetDefaultCollisions _returnvalue_2)
        (bind _returnvalue_3 (TrickTile|GetStaticMeshComponents(Message) (Variables|Default|GetBaseTileCached)))
        (for _array_element _returnvalue_3
          (bind _output_get (LocalVariables|SetMeshComponent _array_element))
          (Utilities|IsValid _output_get
            (:Is%20Valid
              (if _returnvalue
                (Collision|SetCollisionEnabled _meshcomponent_2 _value)
                (else
                  (Collision|SetCollisionEnabled _meshcomponent_2 "QueryAndPhysics"))))))
        (else
          (Collision|SetCollisionEnabled (Variables|/Script/FortniteGame.BuildingSMActor|GetStaticMeshComponent (Variables|Default|GetBaseTileCached)) "QueryAndPhysics"))))
    (:else)))

(fn SetRestoredVisibility
  (bind _basetilecached (Variables|Default|GetBaseTileCached)
    (:then
      (if (Utilities|DoesObjectImplementInterface _basetilecached "/Game/Creative/Devices/TrickTile/TrickTileInterface.TrickTileInterface_C")
        (bind _returnvalue (TrickTile|GetStaticMeshComponents(Message) (Variables|Default|GetBaseTileCached)))
        (for _array_element _returnvalue
          (Utilities|IsValid _array_element
            (:Is%20Valid
              (Rendering|SetVisibility _array_element true false))))
        (else
          (Rendering|SetVisibility (Variables|/Script/FortniteGame.BuildingSMActor|GetStaticMeshComponent (Variables|Default|GetBaseTileCached)) true false))))
    (:else)))

(fn ShouldTriggerTrap
  (return (Math|Boolean|ANDBoolean (Class|CreativeEnabledComponent|IsEnabled (Variables|Default|GetCreative_Enabled_Component)) (Utilities|Enum|NotEqual(Enum) (Variables|Replicated|GetForcedTileState) "NewEnumerator2") (Utilities|Enum|NotEqual(Enum) (Variables|Replicated|GetForcedTileState) "NewEnumerator3"))))

(fn StartActivation
  (bind _returnvalue (Utilities|Time|SetTimerbyEvent (EventDispatchers|CreateEvent :!event OnActivationFinish) (Variables|UserKnob|GetActivationDelayKnob)))
  (Variables|Default|SetActivationTimerHandler _returnvalue)
  (Networking|SwitchHasAuthority
    (:Authority
      (Variables|Replicated|SetCanTriggerOnPlayerContact false))
    (:Remote
      (if (Utilities|Enum|NotEqual(Enum) (Variables|UserKnob|GetTriggerFeedback) "NewEnumerator3")
        (Variables|VFX|SetActivation%20Stage 0)
        (Variables|VFX|SetCurrentActivationTime 0.0)
        (VFX|StartActivationVFX)))))

(fn StartActivationVFX
  (CallFunction|StartVFXTimeline)
  (Utilities|FlowControl|Sequence
    (:then_0
      (if (or (Utilities|Enum|Equal(Enum) (Variables|UserKnob|GetTriggerFeedback) "NewEnumerator0") (Utilities|Enum|Equal(Enum) (Variables|UserKnob|GetTriggerFeedback) "NewEnumerator2"))
        (bind _basetilecached (Variables|Default|GetBaseTileCached)
          (:then
            (if (not (CallFunction|CanDoBuildingAnimations _basetilecached))
              (for _array_element (Rendering|Material|GetMaterials (Building|GetBuildingMeshComponent (Variables|Default|GetBaseTileCached)))
                (bind _asmaterial_instance_dynamic (Utilities|Casting|CastToMaterialInstanceDynamic _array_element))
                (Rendering|Material|SetVectorParameterValue _asmaterial_instance_dynamic "TimestampPayload" (Math|Color|MakeColor 0.0 0.0 (+ (Utilities|Time|GetGameTimeinSeconds) 3.6e+04)) :!class /Script/Engine.MaterialInstanceDynamic)
                (Rendering|Material|SetVectorParameterValue _asmaterial_instance_dynamic "BlueprintColor" "(R=0.000000,G=0.000000,B=0.000000,A=1.000000)" :!class /Script/Engine.MaterialInstanceDynamic))))
          (:else))))
    (:then_1
      (if (or (Utilities|Enum|Equal(Enum) (Variables|UserKnob|GetTriggerFeedback) "NewEnumerator0") (Utilities|Enum|Equal(Enum) (Variables|UserKnob|GetTriggerFeedback) "NewEnumerator1"))
        (VFX|PlayCustomHitBounceAnimation)))))

(fn TryResetTile
  (Utilities|Time|ClearandInvalidateTimerbyHandle (Variables|Default|GetResetAfterTimerHandle))
  (if (Math|Byte|Byte>Byte (Variables|Replicated|GetForcedTileState) (Utilities|Enum|LiteralenumEnum_TrickTileState "NewEnumerator1"))
    (TileState|SetForcedTileState "NewEnumerator1")
    (CallFunction|RestoreTileCanbeDamaged)))

(fn TryTriggerTrap
  (params (Attempting%20Controller :type Controller%20Object%20Reference))
  (if (Trigger|ShouldTriggerTrap)
    (Variables|Trigger|SetTrap%20Instigator Attempting%20Controller)
    (if (> (Variables|UserKnob|GetActivationDelayKnob) 0.0)
      (TileState|SetForcedTileState "NewEnumerator3")
      (else
        (TileState|SetForcedTileState "NewEnumerator2")))
    (Networking|SwitchHasAuthority
      (:Authority
        (GameplayEvent|BroadcastEvent (Variables|UserOptions|Events|GetOn%20Triggered) 0 (Utilities|Struct|MakeGameplayEventTriggered (Variables|Trigger|GetTrap%20Instigator)))))))

(fn UpdateActivationVFXTimes
  (if (> (Variables|UserKnob|GetActivationDelayKnob) 0.0)
    (Variables|VFX|SetStageOneTime (- (Variables|UserKnob|GetActivationDelayKnob) (+ (Variables|VFX|GetDesiredStageOneDuration) (Variables|VFX|GetDesiredStageTwoDuration))))
    (Variables|VFX|SetStageTwoTime (- (Variables|UserKnob|GetActivationDelayKnob) (Variables|VFX|GetDesiredStageTwoDuration)))
    (bind _vfxtimeline (Variables|BP_TrickTile|GetVFXTimeline)
      (:then
        (Components|Timeline|SetPlayRate _vfxtimeline (/ 1.0 (Variables|UserKnob|GetActivationDelayKnob))))
      (:else))))

(fn ConstructionScript)
