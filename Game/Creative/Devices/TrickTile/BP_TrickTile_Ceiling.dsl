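; Ceiling variant of the Creative trick tile device Blueprint.
; It overrides the bounce hit normal to return a constant +Z vector.
; Its construction script only calls the parent implementation, so all other behavior comes from the parent class.
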(fn GetHitNormalForBounce
  (return "0, 0,1.000000"))

(fn ConstructionScript
  (|CallParentFunction))
