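; Blueprint for the wall variant of a Creative Trick Tile device.
; It overrides the bounce normal lookup to return the normalized right vector of the cached base tile actor, so bounces push perpendicular to the wall.
; The construction script only calls the parent, leaving all other behavior to the parent Trick Tile class.
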(fn GetHitNormalForBounce
  (return (Math|Vector|Normalize (Transformation|GetActorRightVector (Variables|Default|GetBaseTileCached)))))

(fn ConstructionScript
  (|CallParentFunction))
