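; TrickTileInterface is an interface for Trick Tile creative devices, judging by its name and folder.
; It declares two functions, GetDefaultCollisions and GetStaticMeshComponents.
; Both contain only empty default returns, with no logic of their own.
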(fn GetDefaultCollisions
  (return 0))

(fn GetStaticMeshComponents
  (return 0))
